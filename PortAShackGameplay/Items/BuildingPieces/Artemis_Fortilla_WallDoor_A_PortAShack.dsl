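; Blueprint for a Port-A-Shack wall-with-door building piece in the Artemis Fortilla set, with player banner support.
; When the building actor initializes, it runs the parent logic, then calls InitBannerPlayer on its ENVPlayerBanner, passing itself and its static mesh component.
; The construction script only defers to the parent.
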(event Building|EventOnBuildingActorInitialized
  (params (InitializationReason :type EFortBuildingInitializationReason%20Enum :default "StaticallyPlaced") (BuildingPersistentState :type EFortBuildingPersistentState%20Enum :default "Default"))
  (|CallParentFunction InitializationReason BuildingPersistentState)
  (Class|ENVPlayerBanner|InitBannerPlayer (Variables|Default|GetENVPlayerBanner) self (Variables|Debug|GetStaticMeshComponent)))

(fn ConstructionScript
  (|CallParentFunction))
