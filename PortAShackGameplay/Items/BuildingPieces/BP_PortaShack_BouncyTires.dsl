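; Port-a-Shack bouncy tire building piece that launches overlapping pawns, self-destructing if disabled and clearing its owner instigator so the owner's pickaxe can damage it.
; Overlap skips recently launched pawns, and downed ones unless allowed, and detaches remote-controlled pawns.
; Launch direction blends the right vector with mirrored velocity, reducing lift for upward movement.
; It then applies a launched effect and calls wobble-cue and force-feedback on the cosmetic check's False branch.
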(event Custom|bounce%20feedback
  (params (EffectCauser :type Actor%20Object%20Reference) (Location :type Vector))
  (bind _overlapcontroller (Variables|Logic|GetOverlapController))
  ;[1 1 1] Bounce Feedback
  (GameplayCue|ExecuteGameplayCue%28Burst%29Local self (Variables|Feedback|GetGC_Bounce) (Ability|GameplayCue|MakeGameplayCueParameters :NormalizedMagnitude (ScalableFloat|GetValueatLevel (Variables|Parameters|GetTireLaunchFeedbackWobbleRadius)) :RawMagnitude 0.0 :OriginalTag (Variables|Feedback|GetGC_Bounce) :Location Location :Normal (* (Transformation|GetActorForwardVector) (ScalableFloat|GetValueatLevel (Variables|Parameters|GetTireLaunchFeedbackWobbleIntensity))) :EffectCauser EffectCauser :GameplayEffectLevel 1 :AbilityLevel 1 :bReplicateLocationWhenUsingMinimalRepProxy false))
  (Utilities|IsValid _overlapcontroller
    (:Is%20Valid
      (Game|Feedback|LocalPlayForceFeedback _overlapcontroller (Variables|Feedback|GetLaunchForceFeedback) false "TireLaunch"))))

(event EventBeginPlay
  (|CallParentFunction)
  (if (ScalableFloat|GetValueasBool (Variables|Parameters|GetTiresEnabled))
    (Networking|SwitchHasAuthority
      (:Authority
        (AbilitySystem|ForceChangeOwnerInstigator 0) ; Change ownership to none. This fixes a bug where this building couldn't be damaged by the owning player's pickax.
      ))
    (else
      (Actor|DestroyActor))))

(event OnComponentBeginOverlap(bounceCol)
  (params (OverlappedComponent :type Primitive%20Component%20Object%20Reference) (OtherActor :type Actor%20Object%20Reference) (OtherComp :type Primitive%20Component%20Object%20Reference) (OtherBodyIndex :type Integer) (bFromSweep :type Boolean) (SweepResult :type Hit%20Result%20Structure%20%28by%20ref%29))
  (bind _tirelaunchverticalcomponent (Variables|Parameters|GetTireLaunchVerticalComponent))
  (bind _returnvalue (ScalableFloat|GetValueatLevel _tirelaunchverticalcomponent))
  (bind _returnvalue_2 (Transformation|GetActorRightVector))
  (bind _returnvalue_3 (Math|Vector|VectorUp))
  (bind _asfort_pawn (Utilities|Casting|CastToFortPawn OtherActor))
  (bind _output_get (Variables|Logic|SetOverlapPawn _asfort_pawn))
  (if (not (or (GameplayTags|HasMatchingGameplayTag (Ability|GetAbilitySystemComponent _output_get) "(TagName=\"Gameplay.Action.Player.Launched\")") (and (Variables|/Script/FortniteGame.FortPawn|GetbIsDBNO _output_get) (not (ScalableFloat|GetValueasBool (Variables|Parameters|GetTireCanBounceDBNO))))))
    (bind _asfort_player_controller (Utilities|Casting|CastToFortPlayerController (Pawn|GetController _output_get))
      (:then
        (Variables|Logic|SetOverlapController _asfort_player_controller))
      (:CastFailed
        _))
    (bind _asfort_player_pawn (Utilities|Casting|CastToFortPlayerPawn (Variables|Logic|GetOverlapPawn))
      (:then
        (RemoteControlPawn|ForceDetachfromRemoteControlledPawn _asfort_player_pawn))
      (:CastFailed
        _))
    ; Debug Display
    (if (Variables|Default|GetDEBUG)
      (Rendering|Debug|DrawDebugArrow (Transformation|GetActorLocation _asfort_player_pawn) (+ (Transformation|GetActorLocation _asfort_player_pawn) (* (Math|Vector|Normalize (Pawn|Components|CharacterMovement|GetLastUpdateVelocity (Variables|/Script/Engine.Character|GetCharacterMovement _asfort_player_pawn))) 2e+02)) 5.0 "(R=1.000000,G=1.000000,B=1.000000,A=1.000000)" 5.0 2.0)
      (Development|PrintString (Utilities|String|Append "Bounce Dot: " (Utilities|String|ToString(Float) 0.0)))
      (Rendering|Debug|DrawDebugArrow (Transformation|GetActorLocation _asfort_player_pawn) (+ (Transformation|GetActorLocation _asfort_player_pawn) (* (Math|Vector|Normalize (Math|Vector|MakeVector (.x (Math|Vector|Lerp(Vector) _returnvalue_2 (Math|Vector|MirrorVectorbyNormal (Math|Vector|Normalize (Pawn|Components|CharacterMovement|GetLastUpdateVelocity (Variables|/Script/Engine.Character|GetCharacterMovement _asfort_player_pawn))) _returnvalue_2) (Math|Float|Clamp(Float) (* (Math|Vector|DotProduct (Math|Vector|Normalize (Pawn|Components|CharacterMovement|GetLastUpdateVelocity (Variables|/Script/Engine.Character|GetCharacterMovement _asfort_player_pawn))) _returnvalue_2) -2.0)))) (.y (Math|Vector|Lerp(Vector) _returnvalue_2 (Math|Vector|MirrorVectorbyNormal (Math|Vector|Normalize (Pawn|Components|CharacterMovement|GetLastUpdateVelocity (Variables|/Script/Engine.Character|GetCharacterMovement _asfort_player_pawn))) _returnvalue_2) (Math|Float|Clamp(Float) (* (Math|Vector|DotProduct (Math|Vector|Normalize (Pawn|Components|CharacterMovement|GetLastUpdateVelocity (Variables|/Script/Engine.Character|GetCharacterMovement _asfort_player_pawn))) _returnvalue_2) -2.0)))) (Math|Float|Lerp _returnvalue 0.0 (Math|Float|Max(Float) (Math|Vector|DotProduct (Math|Vector|Normalize (Pawn|Components|CharacterMovement|GetLastUpdateVelocity (Variables|/Script/Engine.Character|GetCharacterMovement _asfort_player_pawn))) _returnvalue_3))))) 2e+02)) 5.0 "(R=1.000000,G=0.000000,B=0.029169,A=1.000000)" 5.0 2.0)
      (else
        _))
    (Character|LaunchCharacter (Variables|Logic|GetOverlapPawn) (* (Math|Vector|Normalize (Math|Vector|MakeVector (.x (Math|Vector|Lerp(Vector) _returnvalue_2 (Math|Vector|MirrorVectorbyNormal (Math|Vector|Normalize (Pawn|Components|CharacterMovement|GetLastUpdateVelocity (Variables|/Script/Engine.Character|GetCharacterMovement _asfort_player_pawn))) _returnvalue_2) (Math|Float|Clamp(Float) (* (Math|Vector|DotProduct (Math|Vector|Normalize (Pawn|Components|CharacterMovement|GetLastUpdateVelocity (Variables|/Script/Engine.Character|GetCharacterMovement _asfort_player_pawn))) _returnvalue_2) -2.0)))) (.y (Math|Vector|Lerp(Vector) _returnvalue_2 (Math|Vector|MirrorVectorbyNormal (Math|Vector|Normalize (Pawn|Components|CharacterMovement|GetLastUpdateVelocity (Variables|/Script/Engine.Character|GetCharacterMovement _asfort_player_pawn))) _returnvalue_2) (Math|Float|Clamp(Float) (* (Math|Vector|DotProduct (Math|Vector|Normalize (Pawn|Components|CharacterMovement|GetLastUpdateVelocity (Variables|/Script/Engine.Character|GetCharacterMovement _asfort_player_pawn))) _returnvalue_2) -2.0)))) (Math|Float|Lerp _returnvalue 0.0 (Math|Float|Max(Float) (Math|Vector|DotProduct (Math|Vector|Normalize (Pawn|Components|CharacterMovement|GetLastUpdateVelocity (Variables|/Script/Engine.Character|GetCharacterMovement _asfort_player_pawn))) _returnvalue_3))))) (ScalableFloat|GetValueatLevel (Variables|Parameters|GetTireLaunchVelocity))) true false)
    (GameplayEffects|ApplyGameplayEffectToSelf :self (Ability|GetAbilitySystemComponent (Variables|Logic|GetOverlapPawn)) :GameplayEffectClass (Variables|Parameters|GetGE_PlayerLaunched) :Level 0.0)
    (Utilities|FlowControl|Sequence
      (:then_0
        (Networking|CanExecuteCosmeticEvents
          (:False
            (CallFunction|BounceFeedback (Variables|Logic|GetOverlapPawn) (Transformation|GetActorLocation (Variables|Logic|GetOverlapPawn))))))
      (:then_1
        (Variables|Logic|SetOverlapPawn 0)
        (Variables|Logic|SetOverlapController 0)))))

(fn ConstructionScript
  (|CallParentFunction))
