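; Component that shows the owning player's banner on Port-A-Shack building piece meshes.
; InitBannerPlayer stores the mesh and, if enabled, copies banner icon and color IDs from the owner's pawn loadout; otherwise it zeroes color alpha.
; BeginPlay and OnReps resolve those IDs, with default fallbacks, only where cosmetic events can run.
; The icon loads asynchronously into the mesh's dynamic material, which also takes the colors and flips the banner on mirrored meshes.
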(event EventBeginPlay
  ;[0.2082 1 0] BeginPlay. Update the banner on the client immediately
  (|CallParentFunction)
  (CallFunction|UpdateBannerIconId)
  (CallFunction|UpdateBannerColorId))

(event Custom|DisableBanner
  ;[0.2001 1 0] OnBuildingActorInitialized.
  ; Setup the BannerPlayerNetId and clear the owner.
  (Fortnite|FortShippingLog self "DisableBanner() DISASBLED on Port-A-Shack")
  (CallFunction|SetBannerColors "(R=1.000000,G=1.000000,B=1.000000,A=0.000000)" "(R=1.000000,G=1.000000,B=1.000000,A=0.000000)") ; set the alpha of the colors to 0 to disable the banner visuals
)

(event Custom|InitBannerPlayer
  (params (Target%20Building :type Building%20Actor%20Object%20Reference) (StaticMeshComponent :type Static%20Mesh%20Component%20Object%20Reference))
  (bind _returnvalue (Building|GetOwningPlayerState Target%20Building))
  (bind (_ownername _playername) (CallFunction|GetOwnerAndPlayerName))
  (bind (_ownername_2 _playername_2) (CallFunction|GetOwnerAndPlayerName))
  (bind _self self)
  ;[0.2001 1 0] OnBuildingActorInitialized.
  ; Setup the BannerPlayerNetId and clear the owner.
  (Variables|Default|SetStaticMesh StaticMeshComponent)
  (if (ScalableFloat|GetValueasBool (Variables|Default|GetBannerEnabled))
    (Fortnite|FortShippingLog self (Utilities|String|Append "InitBannerPlayer() for owner: [" _ownername_2 "] player: [" _playername_2 "]."))
    (Utilities|IsValid _returnvalue
      (:Is%20Valid
        (bind _asfort_player_pawn_athena (Utilities|Casting|CastToFortPlayerPawnAthena (PlayerState|GetPawn _returnvalue))
          (:then
            (Variables|Default|SetBanner%20Icon%20Id (CosmeticLoadout|GetBannerIconId _asfort_player_pawn_athena))
            (Variables|Default|SetBanner%20Color%20Id (CosmeticLoadout|GetBannerColorId _asfort_player_pawn_athena)))
          (:CastFailed
            ;;[1 0 0.03458] Failure States. Some necessary data is missing.
            (Fortnite|FortShippingLog _self (Utilities|String|Append "InitBannerPlayer FAILED for owner: [" _ownername "] player: [" _playername "], no player pawn.") true))))
      (:Is%20Not%20Valid
        (Fortnite|FortShippingLog _self (Utilities|String|Append "InitBannerPlayer FAILED for owner: [" _ownername "], no player state.") true)))
    (else
      (CallFunction|DisableBanner))))

(event Custom|SetBannerIcon
  (params (AssetToLoad :type Soft%20Object%20Reference))
  (bind (_ownername _playername) (CallFunction|GetOwnerAndPlayerName))
  (bind _staticmesh (Variables|Default|GetStaticMesh))
  (bind (_returnvalue_x _returnvalue_y _returnvalue_z) (Math|Vector|BreakVector (Transformation|GetWorldScale _staticmesh)))
  ;[0.4404 1 0] LoadBannerIcon
  (bind _loadedasset (BannerAsyncLoadAsset|AsyncLoadBannerAsset AssetToLoad)
    (:BannerAssetLoadComplete
      (Fortnite|FortShippingLog self (Utilities|String|Append "SetBannerIcon() for Port-A-Shack mesh:  [" (Utilities|GetDisplayName (Variables|Default|GetStaticMesh)) "]     Owner: [" _ownername "]        Player: [" _playername "]        Banner Texture: [" (Utilities|GetDisplayName _loadedasset) "]."))
      (bind _astexture (Utilities|Casting|CastToTexture _loadedasset))
      (bind _banner_dynamic_material_cached (Utilities|GetBannerDynamicMaterial _staticmesh)
        (:Valid
          (Rendering|Material|SetTextureParameterValue _banner_dynamic_material_cached "EmblemImage_BP" _astexture)
          ;;[0.8494 0 1] If the mesh is flipped, flip the banner.
          (if (< (Utilities|Operators|Multiply _returnvalue_x _returnvalue_y _returnvalue_z) 0)
            (Rendering|Material|SetScalarParameterValue _banner_dynamic_material_cached "FlipBanner" 1.0 :!class /Script/Engine.MaterialInstanceDynamic)))))))

(event Custom|SetBannerColors
  (params (Icon%20Color :type Linear%20Color%20Structure) (Banner%20Color :type Linear%20Color%20Structure))
  (bind (_ownername _playername) (CallFunction|GetOwnerAndPlayerName))
  (bind _staticmesh (Variables|Default|GetStaticMesh))
  ;[0.9543 0 1] Set Banner Color
  (bind _banner_dynamic_material_cached (Utilities|GetBannerDynamicMaterial _staticmesh)
    (:Valid
      (Rendering|Material|SetVectorParameterValue _banner_dynamic_material_cached "BannerColor_BP" Banner%20Color :!class /Script/Engine.MaterialInstanceDynamic)
      (Rendering|Material|SetVectorParameterValue _banner_dynamic_material_cached "EmblemColor" Icon%20Color :!class /Script/Engine.MaterialInstanceDynamic)
      (Fortnite|FortShippingLog self (Utilities|String|Append "SetBannerColors() for Port-A-Shack mesh:  [" (Utilities|GetDisplayName _staticmesh) "]     Owner: [" _ownername "]     Player: [" _playername "]        Icon Color: [" (Utilities|String|ToString(LinearColor) Icon%20Color) "]        Icon Color: [" (Utilities|String|ToString(LinearColor) Banner%20Color) "].")))))

(fn GetOwnerandPlayerName
  (bind _returnvalue (Components|GetOwner))
  (bind _asbuilding_actor (Utilities|Casting|CastToBuildingActor _returnvalue)
    (:then
      (return (Utilities|GetDisplayName _returnvalue) (Fortnite|GetPlayerNameSafe (Building|GetOwningPlayerState _asbuilding_actor))))
    (:CastFailed
      (return "unkonwn" "unknown"))))

(fn OnRep_BannerColorId
  (CallFunction|UpdateBannerColorId))

(fn OnRep_BannerIconId
  (CallFunction|UpdateBannerIconId))

(fn UpdateBannerColorId
  (bind _banner_color_id (Variables|Default|GetBanner%20Color%20Id))
  (bind _returnvalue (Utilities|String|StringToName _banner_color_id))
  (bind _defaulticoncolor (Variables|Default|GetDefaultIconColor))
  (Networking|CanExecuteCosmeticEvents
    (:True
      (if (!= _returnvalue "")
        (bind _returnvalue_2 (Banner|GetBannerColorfromRowName _returnvalue))
        (CallFunction|SetBannerColors _defaulticoncolor (Banner|GetPrimaryColor _returnvalue_2))
        (else
          (CallFunction|SetBannerColors _defaulticoncolor (Variables|Default|GetDefaultIconColor%20Const)))))))

(fn UpdateBannerIconId
  (Networking|CanExecuteCosmeticEvents
    (:True
      ;[0.07717 0.04473 1] Load Banner Icon
      (bind _returnvalue (Banner|GetBannerLargeIconAssetfromRowName (select (!= (Utilities|String|StringToName (Variables|Default|GetBanner%20Icon%20Id)) "") (Utilities|String|StringToName (Variables|Default|GetBanner%20Icon%20Id)) (Variables|Default|GetDefaultIconName%20Const)) "BannerIcon"))
      (CallFunction|SetBannerIcon _returnvalue))))
